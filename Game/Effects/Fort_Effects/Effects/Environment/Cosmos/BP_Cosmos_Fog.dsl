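; BP_Cosmos_Fog is an environment effect Blueprint for the Cosmos fog, built around a haze static mesh component named SM_NoTickHaze.
; On BeginPlay it checks whether the platform is Nintendo Switch and, if so, destroys that haze component.
; The overlap, tick and construction script entry points are empty.
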(event EventBeginPlay
  (if (TouchPlatform|IsNintendoSwitch)
    (Components|DestroyComponent (Variables|Default|GetSM_NoTickHaze))))

(event Collision|EventActorBeginOverlap
  (params (OtherActor :type Actor%20Object%20Reference)))

(event EventTick
  (params (DeltaSeconds :type Float%20%28single-precision%29)))

(fn ConstructionScript)
