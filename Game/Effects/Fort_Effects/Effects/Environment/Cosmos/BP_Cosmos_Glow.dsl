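; BP_Cosmos_Glow is a glow effect actor from the Cosmos environment effects, wrapping a Niagara system with editable scale and color.
; The Construction Script passes the GlowScale and GlowColor variables into the NS_Glow component's User.GlowScale and User.GlowColor parameters.
; BeginPlay, Tick and overlap events are empty, so the Blueprint adds no runtime logic.
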(event EventBeginPlay)

(event Collision|EventActorBeginOverlap
  (params (OtherActor :type Actor%20Object%20Reference)))

(event EventTick
  (params (DeltaSeconds :type Float%20%28single-precision%29)))

(fn ConstructionScript
  (bind _ns_glow (Variables|Default|GetNS_Glow))
  (Niagara|SetNiagaraVariable(Float) _ns_glow "User.GlowScale" (Variables|Default|GetGlowScale))
  (Niagara|SetNiagaraVariable(LinearColor) _ns_glow "User.GlowColor" (Variables|Default|GetGlowColor)))
