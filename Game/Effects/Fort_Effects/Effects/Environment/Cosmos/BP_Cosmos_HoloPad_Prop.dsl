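; Cosmos holo pad environment prop that switches off its hologram display when something enters its trigger volume and restores it when the overlap ends.
; Begin overlap hides the soldier mannequin, deactivates Niagara effects and fades glow from a bright flash to zero before hiding the platform hologram; end overlap reverses this, settling at base brightness.
; Construction creates dynamic materials, tints the stage light blue and caches glow brightness.
; Death fades out the ambient audio loop.
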(event EventBeginPlay)

(event Collision|EventActorBeginOverlap
  (params (OtherActor :type Actor%20Object%20Reference)))

(event EventTick
  (params (DeltaSeconds :type Float%20%28single-precision%29)))

(event EventOnDeathPlayEffects
  (params (Damage :type Float%20%28single-precision%29) (DamageTags :type Gameplay%20Tag%20Container%20Structure%20%28by%20ref%29) (Momentum :type Vector :default "0, 0, 0") (HitInfo :type Hit%20Result%20Structure%20%28by%20ref%29) (InstigatedBy :type Fort%20Pawn%20Object%20Reference) (DamageCauser :type Actor%20Object%20Reference) (EffectContext :type Gameplay%20Effect%20Context%20Handle%20Structure))
  (Audio|Components|Audio|FadeOut (Variables|Default|GetAudio_Amb_Loop) 0.5 0.0 "Linear" :!class /Script/Engine.AudioComponent))

(event OnComponentBeginOverlap(vol_HoloPad)
  (params (OverlappedComponent :type Primitive%20Component%20Object%20Reference) (OtherActor :type Actor%20Object%20Reference) (OtherComp :type Primitive%20Component%20Object%20Reference) (OtherBodyIndex :type Integer) (bFromSweep :type Boolean) (SweepResult :type Hit%20Result%20Structure%20%28by%20ref%29))
  (Rendering|SetVisibility (Variables|Default|GetSoldierMannequin_Male) false false)
  (Components|Activation|Deactivate (Variables|Default|GetNS_Cosmos_HoloPad))
  (bind (_direction _alpha) (Timeline|anim_HoloGlow_End PlayFromStart)
    (:Update
      (Rendering|Material|SetScalarParameterValue (Variables|Default|GetHologramGlow) "OpacityMult" (Math|Float|Lerp (* (Variables|Default|GetHologramGlow_Brightness) 8.0) 0.0 _alpha) :!class /Script/Engine.MaterialInstanceDynamic))
    (:Finished
      (Rendering|SetVisibility (Variables|Default|GetSM_Galileo_PlatformBig_Hologram) false false))))

(event OnComponentEndOverlap(vol_HoloPad)
  (params (OverlappedComponent :type Primitive%20Component%20Object%20Reference) (OtherActor :type Actor%20Object%20Reference) (OtherComp :type Primitive%20Component%20Object%20Reference) (OtherBodyIndex :type Integer))
  (Rendering|SetVisibility (Variables|Default|GetSM_Galileo_PlatformBig_Hologram) true true)
  (Components|Activation|Activate (Variables|Default|GetNS_Cosmos_HoloPad) false)
  (bind (_direction _alpha) (Timeline|anim_HoloGlow_Start PlayFromStart)
    (:Update
      (Rendering|Material|SetScalarParameterValue (Variables|Default|GetHologramGlow) "OpacityMult" (Math|Float|Lerp (* (Variables|Default|GetHologramGlow_Brightness) 8.0) (Variables|Default|GetHologramGlow_Brightness) _alpha) :!class /Script/Engine.MaterialInstanceDynamic))))

(fn ConstructionScript
  (bind _returnvalue (Rendering|Material|CreateDynamicMaterialInstance (Variables|Default|GetSM_WH_Stage_Light) 0 0 "None" :!class /Script/Engine.PrimitiveComponent))
  (Rendering|Material|SetVectorParameterValue _returnvalue "Emissive" "(R=0.099455,G=0.268281,B=1.000000,A=1.000000)" :!class /Script/Engine.MaterialInstanceDynamic)
  (bind _returnvalue_2 (Rendering|Material|CreateDynamicMaterialInstance (Variables|Default|GetSM_Galileo_PlatformBig_Hologram) 1 0 "None" :!class /Script/Engine.PrimitiveComponent))
  (Variables|Default|SetHologramGlow _returnvalue_2)
  (bind _returnvalue_3 (Rendering|Material|GetScalarParameterValue (Variables|Default|GetHologramGlow) "OpacityMult" :!class /Script/Engine.MaterialInstanceDynamic))
  (Variables|Default|SetHologramGlow_Brightness _returnvalue_3))
