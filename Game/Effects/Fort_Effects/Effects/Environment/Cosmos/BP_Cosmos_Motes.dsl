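; Environment effect actor for Cosmos motes, drawn by a dynamic material instance on its SM_NoTickHaze mesh component.
; The construction script creates that instance and feeds it the actor's variables for opacity, glow contrast, sprite size range and near/far distance fades.
; It also passes morning, day, evening and night colors to the material.
; BeginPlay, overlap and Tick events are empty, so the Blueprint runs no gameplay logic.
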(event EventBeginPlay)

(event Collision|EventActorBeginOverlap
  (params (OtherActor :type Actor%20Object%20Reference)))

(event EventTick
  (params (DeltaSeconds :type Float%20%28single-precision%29)))

(fn ConstructionScript
  (bind _mid_motes (Variables|Material|GetMID_Motes))
  (bind _mid_motes_2 (Variables|Material|GetMID_Motes))
  (bind _mid_motes_3 (Variables|Material|GetMID_Motes))
  (bind _returnvalue (Rendering|Material|CreateDynamicMaterialInstance (Variables|Default|GetSM_NoTickHaze) 0 0 "None" :!class /Script/Engine.PrimitiveComponent))
  (Variables|Material|SetMID_Motes _returnvalue)
  ;[1 1 1] Scalar
  (Rendering|Material|SetScalarParameterValue _mid_motes_2 "Opacity" (Variables|Scalar|GetOpacity) :!class /Script/Engine.MaterialInstanceDynamic)
  (Rendering|Material|SetScalarParameterValue _mid_motes_2 "GlowContrast" (Variables|Scalar|GetGlowContrast) :!class /Script/Engine.MaterialInstanceDynamic)
  (Rendering|Material|SetScalarParameterValue _mid_motes_2 "maxSize" (Variables|Scalar|GetSpriteScaleMax) :!class /Script/Engine.MaterialInstanceDynamic)
  (Rendering|Material|SetScalarParameterValue _mid_motes_2 "MinSizeMult" (Variables|Scalar|GetSpriteScaleMin) :!class /Script/Engine.MaterialInstanceDynamic)
  ;[1 1 1] Scalar: Fade Distances
  (Rendering|Material|SetScalarParameterValue _mid_motes_3 "NearFadeLength (S)" (Variables|Scalar|GetNearFadeLength) :!class /Script/Engine.MaterialInstanceDynamic)
  (Rendering|Material|SetScalarParameterValue _mid_motes_3 "NearFadeOffset (S)" (Variables|Scalar|GetNearFadeOffset) :!class /Script/Engine.MaterialInstanceDynamic)
  (Rendering|Material|SetScalarParameterValue _mid_motes_3 "FarFadeLength (S)" (Variables|Scalar|GetFarFadeLength) :!class /Script/Engine.MaterialInstanceDynamic)
  (Rendering|Material|SetScalarParameterValue _mid_motes_3 "FarFadeOffset (S)" (Variables|Scalar|GetFarFadeOffset) :!class /Script/Engine.MaterialInstanceDynamic)
  ;[1 1 1] Colors
  (Rendering|Material|SetVectorParameterValue _mid_motes "Color_Morning (V3)" (Variables|Colors|GetColor_Morning) :!class /Script/Engine.MaterialInstanceDynamic)
  (Rendering|Material|SetVectorParameterValue _mid_motes "Color_Day (V3)" (Variables|Colors|GetColor_Day) :!class /Script/Engine.MaterialInstanceDynamic)
  (Rendering|Material|SetVectorParameterValue _mid_motes "Color_Evening (V3)" (Variables|Colors|GetColor_Evening) :!class /Script/Engine.MaterialInstanceDynamic)
  (Rendering|Material|SetVectorParameterValue _mid_motes "Color_Night (V3)" (Variables|Colors|GetColor_Night) :!class /Script/Engine.MaterialInstanceDynamic))
