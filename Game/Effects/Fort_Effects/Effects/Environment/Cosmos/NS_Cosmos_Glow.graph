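# Niagara system for a soft, tintable Cosmos environment glow, built from one looping mesh-particle emitter despite its FN_Sprite name.
# It steadily spawns overlapping short-lived particles at a single point, with a point force acting on them.
# User GlowColor sets the tint, while GlowScale drives both mesh size and a dynamic material parameter.
# Alpha fades in and out over each lifetime, scaled by a camera-distance fraction curve alongside emitter distance culling.
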NiagaraSystem /Game/Effects/Fort_Effects/Effects/Environment/Cosmos/NS_Cosmos_Glow
dependencies {usedDataInterfaces [&NiagaraDataInterfaceCurve], usedDynamicInputs [], usedModules [&SystemState]}
emitters {FN_Sprite {
  emitterUpdateScript [
   &EmitterState(Enable Distance Culling=true, MaxDistance=2500, Module.bUseMaxDistance=true, Module.bUseMinDistance=true, Scalability Mode=::NewEnumerator1)
   &SpawnRate(SpawnRate=2, Use Spawn Probability=false)
  ]
  particleSpawnScript [
   /Niagara/Modules/Spawn/Initialization/InitializeParticle(
    Color=$User.GlowColor
    Lifetime=&UniformRangedFloat(Minimum=2, Maximum=2)
    Mesh Scale=&Multiply_VectorByFloat(Vector=<1 1 1>, Float=$User.GlowScale)
    Module.Write Color=true
    Module.Write Lifetime=true
    Module.Write Mass=true
    Module.Write Material Random=true
    Module.Write Position=false
    Module.Write Scale=true
    Module.Write SpriteRotation=false
    Module.Write SpriteSize=true
    Position=&SimulationPosition()
    Sprite Rotation=&UniformRangedFloat(Minimum=0, Maximum=360)
   )
   &AccelerationForce()
   &ApplyInitialForces()
   &SphereLocation(
    MaskSpawnGroup=false
    Override Seed=false
    Sphere Radius=0
    Surface Only=false
    UseNonUniformScale=false
    UseOffset=false
    UseSphereOrientation=false
   )
  ]
  particleUpdateScript [
   &ParticleState()
   &PointForce(
    Force Origin=&Add_Vector(A=<0 0 500>, B=<0 0 500>)
    Force Strength=1500
    Module.Use Falloff=true
    Module.UseLinearForce=false
   )
   &SolveForcesAndVelocity(Clamp Velocity=false, Limit Acceleration=false)
   &DynamicMaterialParameters(
    Index 0 Param 1=&Multiply_Float(A=$User.GlowScale, B=0.7)
    Index 0 Param 2=1
    Index 0 Param 3=1
    Index 0 Param 4=1
    Module.Param0WriteEnabled=true
    Module.Param1WriteEnabled=true
    Module.Param2WriteEnabled=true
    Module.Param3WriteEnabled=true
   )
   &ScaleColor(
    Scale Alpha=&FloatFromCurve(
     FloatCurve=DataInterface() {propertyValues {Curve {keys [
         {interpMode=RCIM_Cubic, tangentMode=RCTM_User, time=0, value=0}
         {arriveTangent=0.002, leaveTangent=0.002, time=0.25, value=1}
         {arriveTangent=-0.0331, interpMode=RCIM_Cubic, leaveTangent=-0.0331, tangentMode=RCTM_User, time=0.75, value=1}
         {interpMode=RCIM_Cubic, tangentMode=RCTM_User, time=1.0011, value=0.002}
        ]}}}
     Scale Curve=&FloatFromCurve(
      FloatCurve=DataInterface() {propertyValues {Curve {keys [{interpMode=RCIM_Cubic, tangentMode=RCTM_User, time=0, value=0}, {arriveTangent=2.0469, leaveTangent=2.0469, time=0.7, value=1}, <at 1 1>]}}}
      CurveIndex=$Emitter.Scalability.DistanceFraction
     )
    )
    ScaleA=true
    ScaleRGB=true
   )
  ]
  rendererClasses [&NiagaraMeshRendererProperties]
  renderers [{rendererClass=&NiagaraMeshRendererProperties, rendererIndex=0}]
 }}
summary {
 systemName=NS_Cosmos_Glow
 emitters [{emitterName=FN_Sprite, rendererClasses [&NiagaraMeshRendererProperties]}]
 userVariables [
  User.GlowColor {type=LinearColor, description="", defaultValue {struct=LinearColor, value=<rgba 0.9583 0.7374 0.4544 0.01>}}
  User.GlowScale {type=NiagaraFloat, defaultValue=13, description=""}
 ]
}
