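# Niagara system for the Cosmos hologram pad effect: a blue, looping projector glow with rising particles.
# GPU sprite emitters draw a flat base glow plus small bits and thin streaks, launched from narrow upward cones and faded by randomized alpha curves.
# A CPU mesh glow orb animates dynamic material parameters, while light particles drift upward with a pulsing brightness curve.
# A component-renderer character emitter exists but is disabled.
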NiagaraSystem /Game/Effects/Fort_Effects/Effects/Environment/Cosmos/NS_Cosmos_HoloPad
dependencies {usedDataInterfaces [&NiagaraDataInterfaceCurve], usedDynamicInputs [], usedModules [&SystemState]}
emitters {
 Base {
  simTarget=GPUComputeSim
  emitterUpdateScript [
   &EmitterState(Life Cycle Mode=::NewEnumerator1, Loop Duration=1, Scalability Mode=::NewEnumerator1, UseLoopDelay=false)
   &SpawnRate(SpawnRate=3, Use Spawn Probability=false) {module=SpawnRate001}
  ]
  particleSpawnScript [
   /Niagara/Modules/Spawn/Initialization/InitializeParticle(
    Color=<rgba 1.2213 3.0848 8 1>
    Lifetime=&UniformRangedFloat(Minimum=1, Maximum=1)
    Mesh Scale=&Multiply_Vector(
     A=<1 1 1>
     B=&VectorFromFloat(Value=&UniformRangedFloat(Minimum=0.2, Maximum=1))
    )
    Module.Write Color=true
    Module.Write Lifetime=true
    Module.Write Mass=true
    Module.Write Material Random=true
    Module.Write Position=false
    Module.Write Scale=false
    Module.Write SpriteRotation=true
    Module.Write SpriteSize=true
    Position=&SimulationPosition()
    Sprite Rotation=&RandomRangeFloat(Minimum=0, Maximum=360)
    Sprite Size=<74 74>
   )
   &SphereLocation(
    MaskSpawnGroup=false
    Offset=<0 0 7>
    Override Seed=false
    Sphere Radius=1
    Surface Only=false
    UseNonUniformScale=false
    UseOffset=true
    UseSphereOrientation=false
   )
   /NiagaraFN/Sprite/FN_SpriteAlignmentAndFacing(SpriteAlignment:Vector3f=<0 1 0>, SpriteFacing:Vector3f=<0 0 1>)
   &AddVelocity(
    Scale Added Velocity=&Lerp_Vector(
     A=<0 0 1>
     B=<0 0 3>
     Alpha=&FloatFromCurve(
      FloatCurve=DataInterface() {propertyValues {Curve {keys [<at 0 0>,<at 1 1>]}}}
      CurveIndex=&Divide_Float(A=&VectorLength(Vector=$Particles.FallingSpeed), B=3000)
     )
    )
    Velocity=<0 0 50>
   ) {enabled=false}
   &AddVelocityFromPoint(Module.Use Falloff Distance=false, Origin Offset=<0 0 -75>, Velocity Strength=50) {enabled=false}
   /SetVariables(Particles.RandAlpha:NiagaraFloat=&RandomRangeFloat(Minimum=0.3, Maximum=1)) {enabled=false, isSetParameters=true, module=SetVariables_#1}
  ]
  particleUpdateScript [
   &ParticleState()
   &SolveForcesAndVelocity(Clamp Velocity=false, Limit Acceleration=false)
   &DynamicMaterialParameters(
    Index 0 Param 1=0.85
    Index 0 Param 2=0.2
    Index 0 Param 3=1
    Index 0 Param 4=1
    Module.Param0WriteEnabled=true
    Module.Param1WriteEnabled=true
    Module.Param2WriteEnabled=false
    Module.Param3WriteEnabled=false
   )
   &ScaleColor(
    Scale Alpha=&FloatFromCurve(
     FloatCurve=DataInterface() {propertyValues {Curve {keys [<at 0 0 RCIM_Cubic>,<at 0.5 1.1 RCIM_Cubic>,<at 1 0 RCIM_Cubic>]}}}
     Scale Curve=$Particles.RandAlpha
    )
    ScaleA=true
    ScaleRGB=true
   ) {enabled=false}
  ]
  rendererClasses [&NiagaraSpriteRendererProperties]
  renderers [{rendererClass=&NiagaraSpriteRendererProperties}]
 }
 Bits {
  simTarget=GPUComputeSim
  emitterUpdateScript [
   &EmitterState(Life Cycle Mode=::NewEnumerator1, Loop Duration=1, Scalability Mode=::NewEnumerator1, UseLoopDelay=false)
   &SpawnRate(SpawnRate=3, Use Spawn Probability=false) {module=SpawnRate001}
  ]
  particleSpawnScript [
   &InitializeParticle(
    Color=<rgba 1.2213 3.0848 8 1>
    Lifetime=&UniformRangedFloat(Minimum=0.25, Maximum=2)
    Mesh Scale=&Multiply_Vector(
     A=<1 1 1>
     B=&VectorFromFloat(Value=&UniformRangedFloat(Minimum=0.2, Maximum=1))
    )
    Module.Write Color=true
    Module.Write Lifetime=true
    Module.Write Mass=true
    Module.Write Material Random=true
    Module.Write Position=false
    Module.Write Scale=false
    Module.Write SpriteRotation=true
    Module.Write SpriteSize=true
    Position=&SimulationPosition()
    Sprite Size=&RandomRangeVector2D(Minimum=<2 2>, Maximum=<3 3>)
   )
   /Niagara/Modules/Spawn/Location/ConeLocation(
    Cone Angle=5
    Cone Axis:Vector3f=<0 0 1>
    Enable Wedge=true
    MaskSpawnGroup=false
    Module.Override Randomness=false
    Offset:Vector3f=<0 0 -20>
   )
   &AddVelocity(
    Scale Added Velocity=&Lerp_Vector(
     A=<0 0 1>
     B=<0 0 3>
     Alpha=&FloatFromCurve(
      FloatCurve=DataInterface() {propertyValues {Curve {keys [<at 0 0>,<at 1 1>]}}}
      CurveIndex=&Divide_Float(A=&VectorLength(Vector=$Particles.FallingSpeed), B=3000)
     )
    )
    Velocity=<0 0 50>
   ) {enabled=false}
   &AddVelocityFromPoint(Module.Use Falloff Distance=false, Origin Offset=<0 0 -25>, Velocity Strength=50)
   /SetVariables(Particles.RandAlpha:NiagaraFloat=&RandomRangeFloat(Minimum=0.3, Maximum=1)) {isSetParameters=true, module=SetVariables_#1}
  ]
  particleUpdateScript [
   &ParticleState()
   &Drag(Use Linear Drag=true, Use Rotational Drag=false)
   &SolveForcesAndVelocity(Clamp Velocity=false, Limit Acceleration=false)
   &ScaleColor(
    Scale Alpha=&FloatFromCurve(
     FloatCurve=DataInterface() {propertyValues {Curve {keys [<at 0 0 RCIM_Cubic>,<at 0.5 1.1 RCIM_Cubic>,<at 1 0 RCIM_Cubic>]}}}
     Scale Curve=$Particles.RandAlpha
    )
    ScaleA=true
    ScaleRGB=true
   )
  ]
  rendererClasses [&NiagaraSpriteRendererProperties]
  renderers [{rendererClass=&NiagaraSpriteRendererProperties}]
 }
 Glow_Orb {
  emitterUpdateScript [
   &EmitterState()
   &SpawnRate(SpawnRate=4, Use Spawn Probability=false)
  ]
  particleSpawnScript [
   &InitializeParticle(
    Color=<rgba 0.0995 0.2683 1 0.1>
    Lifetime=&UniformRangedFloat(Minimum=0.7, Maximum=1.2)
    Mesh Scale=&Multiply_VectorByFloat(Vector=<1 1 1>, Float=1.3)
    Module.Write Color=true
    Module.Write Lifetime=true
    Module.Write Mass=true
    Module.Write Material Random=true
    Module.Write Position=false
    Module.Write Scale=true
    Module.Write SpriteRotation=false
    Module.Write SpriteSize=true
    Position=&SimulationPosition()
    Sprite Rotation=&UniformRangedFloat(Minimum=0, Maximum=360)
   )
   &AccelerationForce()
   &ApplyInitialForces()
   &SphereLocation(
    MaskSpawnGroup=false
    Override Seed=false
    Sphere Radius=0
    Surface Only=false
    UseNonUniformScale=false
    UseOffset=false
    UseSphereOrientation=false
   )
  ]
  particleUpdateScript [
   &ParticleState()
   &PointForce(
    Force Origin=&Add_Vector(A=<0 0 500>, B=<0 0 500>)
    Force Strength=1500
    Module.Use Falloff=true
    Module.UseLinearForce=false
   )
   &SolveForcesAndVelocity(Clamp Velocity=false, Limit Acceleration=false)
   &DynamicMaterialParameters(
    Index 0 Param 1=&Lerp_Float(A=0.65, B=0.55, Alpha=$Particles.NormalizedAge)
    Index 0 Param 2=1
    Index 0 Param 3=10
    Index 0 Param 4=1
    Module.Param0WriteEnabled=true
    Module.Param1WriteEnabled=true
    Module.Param2WriteEnabled=true
    Module.Param3WriteEnabled=true
   )
   &ScaleColor(
    Scale Alpha=&FloatFromCurve(FloatCurve=DataInterface() {propertyValues {Curve {keys [
         {interpMode=RCIM_Cubic, tangentMode=RCTM_User, time=0, value=0}
         {arriveTangent=0.002, leaveTangent=0.002, time=0.25, value=1}
         {arriveTangent=-0.0331, interpMode=RCIM_Cubic, leaveTangent=-0.0331, tangentMode=RCTM_User, time=0.75, value=1}
         {interpMode=RCIM_Cubic, tangentMode=RCTM_User, time=1.0011, value=0.002}
        ]}}})
    ScaleA=true
    ScaleRGB=true
   )
   &ScaleMeshSize() {enabled=false}
  ]
  rendererClasses [&NiagaraMeshRendererProperties]
  renderers [{rendererClass=&NiagaraMeshRendererProperties, rendererIndex=0}]
 }
 Light {
  emitterUpdateScript [
   &EmitterState(Life Cycle Mode=::NewEnumerator1, UseLoopDelay=false) {module=EmitterState001}
   &EmitterState()
   &SpawnRate(Use Spawn Probability=false)
  ]
  particleSpawnScript [
   &InitializeParticle(
    Lifetime=&UniformRangedFloat(Minimum=1.6, Maximum=2)
    Mass=&UniformRangedFloat(Minimum=0.6, Maximum=1.2)
    Module.Write Color=false
    Module.Write Lifetime=true
    Module.Write Mass=false
    Module.Write Material Random=false
    Module.Write Position=false
    Module.Write Scale=false
    Module.Write SpriteRotation=false
    Module.Write SpriteSize=false
    Position=&SimulationPosition()
    Sprite Rotation=&NormalizedAngleToDegrees(NormalizedAngle=&UniformRangedFloat(Minimum=0, Maximum=1))
   )
   &AddVelocity(Velocity=&UniformRangedVector(Minimum=<0 0 35>, Maximum=<0 0 125>))
   /Niagara/Modules/Spawn/Location/SphereLocation(MaskSpawnGroup=false, Module.Override Randomness=false, Offset=<0 0 25>, Sphere Radius=5, Surface Only=false)
  ]
  particleUpdateScript [
   &ParticleState()
   &FN_Light_Attributes(Color=<rgba 0.0718 0.3179 1 1>, Create Color Attribute=true, Create Position Attribute=true, Scale Color=6)
   &ScaleColor(
    Color Value To Scale=$Particles.FN_Light_Attributes.Color
    Scale RGB=&Multiply_VectorByFloat(
     Vector=<1 1 1>
     Float=&FloatFromCurve(FloatCurve=DataInterface() {propertyValues {Curve {keys [
          {interpMode=RCIM_Cubic, tangentMode=RCTM_User, time=0, value=0}
          {interpMode=RCIM_Cubic, tangentMode=RCTM_User, time=0.4, value=1}
          {arriveTangent=-0.2604, interpMode=RCIM_Cubic, leaveTangent=-0.2604, tangentMode=RCTM_User, time=1, value=0}
         ]}}})
    )
    ScaleA=true
    ScaleRGB=true
   )
   &SolveForcesAndVelocity(Clamp Velocity=false, Limit Acceleration=false)
  ]
  rendererClasses [&NiagaraLightRendererProperties]
  renderers [{rendererClass=&NiagaraLightRendererProperties, rendererIndex=0}]
 }
 Streaks {
  simTarget=GPUComputeSim
  emitterUpdateScript [
   &EmitterState(Life Cycle Mode=::NewEnumerator1, Loop Duration=1, Scalability Mode=::NewEnumerator1, UseLoopDelay=false)
   &SpawnRate(SpawnRate=15, Use Spawn Probability=false) {module=SpawnRate001}
  ]
  particleSpawnScript [
   &InitializeParticle(
    Color=<rgba 1.2213 3.0848 8 0.2>
    Lifetime=&UniformRangedFloat(Minimum=0.4, Maximum=0.85)
    Mesh Scale=&Multiply_Vector(
     A=<1 1 1>
     B=&VectorFromFloat(Value=&UniformRangedFloat(Minimum=0.2, Maximum=1))
    )
    Module.Write Color=true
    Module.Write Lifetime=true
    Module.Write Mass=true
    Module.Write Material Random=true
    Module.Write Position=false
    Module.Write Scale=false
    Module.Write SpriteRotation=true
    Module.Write SpriteSize=true
    Position=&SimulationPosition()
    Sprite Rotation=90
    Sprite Size=&MakeVector2D(
     X=&UniformRangedFloat(Minimum=0.5, Maximum=1.5)
     Y=&UniformRangedFloat(Minimum=50, Maximum=150)
    )
   )
   &CylinderLocation(
    Cylinder Height=5
    Cylinder Radius=20
    MaskSpawnGroup=false
    Module.Enable Lathe Profile=false
    Module.Override Randomness=false
    Module.Surface Only=true
    Surface Only Band Thickness=1
   ) {enabled=false}
   &ConeLocation(
    Cone Angle=5
    Cone Axis:Vector3f=<0 0 1>
    Cone Length=30
    Enable Wedge=true
    MaskSpawnGroup=false
    Module.Override Randomness=false
    Offset:Vector3f=<0 0 -10>
    Wedge Vertical Angle=90
   )
   &AddVelocity(
    Scale Added Velocity=&Lerp_Vector(
     A=<0 0 1>
     B=<0 0 3>
     Alpha=&FloatFromCurve(
      FloatCurve=DataInterface() {propertyValues {Curve {keys [<at 0 0>,<at 1 1>]}}}
      CurveIndex=&Divide_Float(A=&VectorLength(Vector=$Particles.FallingSpeed), B=3000)
     )
    )
    Velocity=<0 0 50>
   ) {enabled=false}
   &AddVelocityFromPoint(Module.Use Falloff Distance=false, Origin Offset=<0 0 -30>, Velocity Strength=50)
   /SetVariables(Particles.RandAlpha:NiagaraFloat=&RandomRangeFloat(Minimum=0.3, Maximum=1)) {isSetParameters=true, module=SetVariables_#1}
  ]
  particleUpdateScript [
   &ParticleState()
   &SolveForcesAndVelocity(Clamp Velocity=false, Limit Acceleration=false)
   &ScaleColor(
    Scale Alpha=&FloatFromCurve(
     FloatCurve=DataInterface() {propertyValues {Curve {keys [<at 0 0 RCIM_Cubic>,<at 0.5 1.1 RCIM_Cubic>,<at 1 0 RCIM_Cubic>]}}}
     Scale Curve=$Particles.RandAlpha
    )
    ScaleA=true
    ScaleRGB=true
   )
  ]
  rendererClasses [&NiagaraSpriteRendererProperties]
  renderers [{rendererClass=&NiagaraSpriteRendererProperties}]
 }
 mesh_character {
  bEnabled=false
  emitterUpdateScript [
   &EmitterState()
   &SpawnRate(SpawnRate=2, Use Spawn Probability=false)
  ]
  particleSpawnScript [
   &InitializeParticle(
    Lifetime=&UniformRangedFloat(Minimum=2, Maximum=2)
    Mesh Scale=&Multiply_VectorByFloat(Vector=<1 1 1>, Float=1)
    Module.Write Color=true
    Module.Write Lifetime=true
    Module.Write Mass=true
    Module.Write Material Random=true
    Module.Write Position=false
    Module.Write Scale=true
    Module.Write SpriteRotation=false
    Module.Write SpriteSize=true
    Position=&SimulationPosition()
    Sprite Rotation=&UniformRangedFloat(Minimum=0, Maximum=360)
   )
   &AccelerationForce() {enabled=false}
   &ApplyInitialForces() {enabled=false}
   /Niagara/Modules/Spawn/Location/V2/SphereLocation(
    MaskSpawnGroup=false
    Override Seed=false
    Sphere Radius=1
    Surface Only=false
    UseNonUniformScale=false
    UseOffset=false
    UseSphereOrientation=false
   )
   /SetVariables() {isSetParameters=true}
  ]
  particleUpdateScript [
   &ParticleState()
   &PointForce(
    Force Origin=&Add_Vector(A=<0 0 500>, B=<0 0 500>)
    Force Strength=1500
    Module.Use Falloff=true
    Module.UseLinearForce=false
   )
   &SolveForcesAndVelocity(Clamp Velocity=false, Limit Acceleration=false)
   &DynamicMaterialParameters(
    Index 0 Param 1=&Lerp_Float(A=0.65, B=0.55, Alpha=$Particles.NormalizedAge)
    Index 0 Param 2=1
    Index 0 Param 3=10
    Index 0 Param 4=1
    Module.Param0WriteEnabled=true
    Module.Param1WriteEnabled=true
    Module.Param2WriteEnabled=true
    Module.Param3WriteEnabled=true
   ) {enabled=false}
   &ScaleColor(
    Scale Alpha=&FloatFromCurve(FloatCurve=DataInterface() {propertyValues {Curve {keys [
         {interpMode=RCIM_Cubic, tangentMode=RCTM_User, time=0, value=0}
         {arriveTangent=0.002, leaveTangent=0.002, time=0.25, value=1}
         {arriveTangent=-0.0331, interpMode=RCIM_Cubic, leaveTangent=-0.0331, tangentMode=RCTM_User, time=0.75, value=1}
         {interpMode=RCIM_Cubic, tangentMode=RCTM_User, time=1.0011, value=0.002}
        ]}}})
    ScaleA=true
    ScaleRGB=true
   ) {enabled=false}
   &ScaleMeshSize() {enabled=false}
  ]
  rendererClasses [&NiagaraComponentRendererProperties]
  renderers [
   {rendererClass=&NiagaraComponentRendererProperties, rendererIndex=0}
   {rendererClass=&NiagaraComponentRendererProperties, rendererIndex=1}
  ]
 }
}
summary {
 systemName=NS_Cosmos_HoloPad
 emitters [
  {emitterName=Glow_Orb, rendererClasses [&NiagaraMeshRendererProperties]}
  {emitterName=Streaks, simTarget=GPUComputeSim, rendererClasses [&NiagaraSpriteRendererProperties]}
  {emitterName=Bits, simTarget=GPUComputeSim, rendererClasses [&NiagaraSpriteRendererProperties]}
  {emitterName=Base, simTarget=GPUComputeSim, rendererClasses [&NiagaraSpriteRendererProperties]}
  {emitterName=Light, rendererClasses [&NiagaraLightRendererProperties]}
  {bEnabled=false, emitterName=mesh_character, rendererClasses [&NiagaraComponentRendererProperties]}
 ]
 userVariables []
}
